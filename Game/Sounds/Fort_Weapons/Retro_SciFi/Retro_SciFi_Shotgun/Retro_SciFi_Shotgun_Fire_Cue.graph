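# Sound cue for the Retro SciFi shotgun's firing sound, layered by listener distance.
# Close, medium and distant layers each randomly pick one of five waves, vary it through a modulator and apply their own attenuation.
# The close layer uses the third-person fire sound class, while medium and distant are mixed under the distant fire class.
# Overridden concurrency limits voices per owner, stopping the oldest.
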SoundCue /Game/Sounds/Fort_Weapons/Retro_SciFi/Retro_SciFi_Shotgun/Retro_SciFi_Shotgun_Fire_Cue
root=Mixer_1
nodes [
 {
  id=Attenuation_10
  children [Modulator_14]
  properties {attenuationSettings=&Fort_Weapon_Projectile_Medium_Mono_Athena_Att}
 }
 {
  id=Attenuation_11
  children [Modulator_13]
  properties {attenuationSettings=&Fort_Weapon_Projectile_Distant_Mono_Athena_Att}
 }
 {id=Attenuation_9, children [Modulator_15], properties {attenuationSettings=&Fort_Weapon_Projectile_Close_Athena_Att}}
 {id=Mixer_1, children [SoundClass_7,SoundClass_6], properties {inputVolume [1,1]}}
 {id=Mixer_2, children [Attenuation_10, Attenuation_11], properties {inputVolume [0.7,0.6]}}
 {id=Modulator_13, children [Random_2], properties {pitchMax=1.1, volumeMax=0.8, volumeMin=0.7}}
 {id=Modulator_14, children [Random_1], properties {pitchMax=1.1, volumeMax=0.8, volumeMin=0.7}}
 {id=Modulator_15, children [Random_0], properties {volumeMax=1.3, volumeMin=1.2}}
 {id=Random_0, children [WavePlayer_0,WavePlayer_1,WavePlayer_2,WavePlayer_3,WavePlayer_4], properties {weights [1,1,1,1,1]}}
 {id=Random_1, children [WavePlayer_5, WavePlayer_6, WavePlayer_7, WavePlayer_8, WavePlayer_22], properties {weights [1,1,1,1,1]}}
 {id=Random_2, children [WavePlayer_23, WavePlayer_24, WavePlayer_25, WavePlayer_26, WavePlayer_27], properties {weights [1,1,1,1,1]}}
 {id=SoundClass_6, children [Mixer_2], properties {soundClassOverride=&Fort_WeaponRanged_Fire_Distant}}
 {
  id=SoundClass_7
  children [Attenuation_9]
  properties {soundClassOverride=&Fort_WeaponRanged_Fire_3P_PSM}
 }
 {id=WavePlayer_0, properties {soundWaveAssetPtr=&/Retro_SciFi_Shotgun_Fire_Close_01}}
 {id=WavePlayer_1, properties {soundWaveAssetPtr=&/Retro_SciFi_Shotgun_Fire_Close_02}}
 {id=WavePlayer_2, properties {soundWaveAssetPtr=&/Retro_SciFi_Shotgun_Fire_Close_03}}
 {id=WavePlayer_22, properties {soundWaveAssetPtr=&/Retro_SciFi_Shotgun_Fire_Med_05}}
 {id=WavePlayer_23, properties {soundWaveAssetPtr=&/Retro_SciFi_Shotgun_Fire_Dist_01}}
 {id=WavePlayer_24, properties {soundWaveAssetPtr=&/Retro_SciFi_Shotgun_Fire_Dist_02}}
 {id=WavePlayer_25, properties {soundWaveAssetPtr=&/Retro_SciFi_Shotgun_Fire_Dist_03}}
 {id=WavePlayer_26, properties {soundWaveAssetPtr=&/Retro_SciFi_Shotgun_Fire_Dist_04}}
 {id=WavePlayer_27, properties {soundWaveAssetPtr=&/Retro_SciFi_Shotgun_Fire_Dist_05}}
 {id=WavePlayer_3, properties {soundWaveAssetPtr=&/Retro_SciFi_Shotgun_Fire_Close_04}}
 {id=WavePlayer_4, properties {soundWaveAssetPtr=&/Retro_SciFi_Shotgun_Fire_Close_05}}
 {id=WavePlayer_5, properties {soundWaveAssetPtr=&/Retro_SciFi_Shotgun_Fire_Med_01}}
 {id=WavePlayer_6, properties {soundWaveAssetPtr=&/Retro_SciFi_Shotgun_Fire_Med_02}}
 {id=WavePlayer_7, properties {soundWaveAssetPtr=&/Retro_SciFi_Shotgun_Fire_Med_03}}
 {id=WavePlayer_8, properties {soundWaveAssetPtr=&/Retro_SciFi_Shotgun_Fire_Med_04}}
]
properties {
 bOverrideConcurrency=true
 maxDistance=27500
 volumeMultiplier=0.5
 attenuationOverrides {falloffDistance=5000, attenuationShapeExtents {x=1000}}
 concurrencyOverrides {bLimitToOwner=true, maxCount=3, resolutionRule=StopOldest}
}
