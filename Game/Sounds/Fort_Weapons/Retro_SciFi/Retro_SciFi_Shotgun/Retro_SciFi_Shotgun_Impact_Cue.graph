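# Sound cue for retro sci-fi shotgun impacts, routed to the Fort_Impacts sound class.
# A Random node picks among six equally weighted impact waves and feeds a Modulator whose pitch range is fixed at unity.
# It uses a long-range bullet-impact-on-opponent attenuation asset and overrides concurrency with a small instance cap that blocks new plays once full.
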SoundCue /Game/Sounds/Fort_Weapons/Retro_SciFi/Retro_SciFi_Shotgun/Retro_SciFi_Shotgun_Impact_Cue
root=Modulator_5
nodes [
 {id=Modulator_5, children [Random_1], properties {pitchMax=1, pitchMin=1}}
 {id=Random_1, children [WavePlayer_3,WavePlayer_4,WavePlayer_5,WavePlayer_6,WavePlayer_7,WavePlayer_8], properties {weights [1,1,1,1,1,1]}}
 {id=WavePlayer_3, properties {soundWaveAssetPtr=&/Retro_SciFi_Shotgun_Impact_01}}
 {id=WavePlayer_4, properties {soundWaveAssetPtr=&/Retro_SciFi_Shotgun_Impact_02}}
 {id=WavePlayer_5, properties {soundWaveAssetPtr=&/Retro_SciFi_Shotgun_Impact_03}}
 {id=WavePlayer_6, properties {soundWaveAssetPtr=&/Retro_SciFi_Shotgun_Impact_04}}
 {id=WavePlayer_7, properties {soundWaveAssetPtr=&/Retro_SciFi_Shotgun_Impact_05}}
 {id=WavePlayer_8, properties {soundWaveAssetPtr=&/Retro_SciFi_Shotgun_Impact_06}}
]
properties {
 attenuationSettings=/Game/Sounds/Attenuation/Fort_Bullet_ImpactOpponent_Lng_Att
 bOverrideConcurrency=true
 maxDistance=5500
 soundClassObject=&Fort_Impacts
 volumeMultiplier=1
 concurrencyOverrides {maxCount=3, resolutionRule=PreventNew}
}
