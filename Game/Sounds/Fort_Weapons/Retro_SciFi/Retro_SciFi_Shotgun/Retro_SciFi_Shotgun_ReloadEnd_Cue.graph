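# Sound cue for the end-of-reload sound of the Retro SciFi shotgun.
# A Random node picks one of three equally weighted ReloadEnd wave variations and feeds it through a root Modulator with default settings for variation.
# The cue plays at reduced volume in the weapon foley sound class, using the shared weapon reload attenuation.
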SoundCue /Game/Sounds/Fort_Weapons/Retro_SciFi/Retro_SciFi_Shotgun/Retro_SciFi_Shotgun_ReloadEnd_Cue
root=Modulator_0
nodes [
 {id=Modulator_0, children [Random_2]}
 {id=Random_2, children [WavePlayer_1,WavePlayer_7,WavePlayer_8], properties {weights [1,1,1]}}
 {id=WavePlayer_1, properties {soundWaveAssetPtr=&/Retro_SciFi_Shotgun_ReloadEnd_01}}
 {id=WavePlayer_7, properties {soundWaveAssetPtr=&/Retro_SciFi_Shotgun_ReloadEnd_02}}
 {id=WavePlayer_8, properties {soundWaveAssetPtr=&/Retro_SciFi_Shotgun_ReloadEnd_03}}
]
properties {
 attenuationSettings=&Fort_Weapon_Reload_Att
 maxDistance=2250
 soundClassObject=&Fort_WeaponFoley
 volumeMultiplier=0.15
}
