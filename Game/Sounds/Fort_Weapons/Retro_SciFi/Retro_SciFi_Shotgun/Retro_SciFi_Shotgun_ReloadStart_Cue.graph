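# Sound cue for the start of the retro sci-fi shotgun reload, classed as weapon foley with reload attenuation.
# A mixer layers a random reload-start sample with a random reload-insert sample that is slightly louder in the mix.
# The insert layer runs through a fixed short delay and a modulator that lowers its pitch, so it follows the start sound as a deeper shell-insert accent.
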SoundCue /Game/Sounds/Fort_Weapons/Retro_SciFi/Retro_SciFi_Shotgun/Retro_SciFi_Shotgun_ReloadStart_Cue
root=Mixer_0
nodes [
 {id=Delay_2, children [Modulator_0], properties {delayMax=0.2, delayMin=0.2}}
 {id=Mixer_0, children [Random_2,Delay_2], properties {inputVolume [1,1.2]}}
 {id=Modulator_0, children [Random_3], properties {pitchMax=0.8, pitchMin=0.8}}
 {id=Random_2, children [WavePlayer_3,WavePlayer_8,WavePlayer_9], properties {weights [1,1,1]}}
 {id=Random_3, children [WavePlayer_10, WavePlayer_11, WavePlayer_12], properties {weights [1,1,1]}}
 {
  id=WavePlayer_10
  properties {soundWaveAssetPtr=&/Retro_SciFi_Shotgun_ReloadInsert_01}
 }
 {
  id=WavePlayer_11
  properties {soundWaveAssetPtr=&/Retro_SciFi_Shotgun_ReloadInsert_02}
 }
 {
  id=WavePlayer_12
  properties {soundWaveAssetPtr=&/Retro_SciFi_Shotgun_ReloadInsert_03}
 }
 {id=WavePlayer_3, properties {soundWaveAssetPtr=&/Retro_SciFi_Shotgun_ReloadStart_01}}
 {id=WavePlayer_8, properties {soundWaveAssetPtr=&/Retro_SciFi_Shotgun_ReloadStart_02}}
 {id=WavePlayer_9, properties {soundWaveAssetPtr=&/Retro_SciFi_Shotgun_ReloadStart_03}}
]
properties {
 attenuationSettings=&Fort_Weapon_Reload_Att
 maxDistance=2250
 soundClassObject=&Fort_WeaponFoley
 volumeMultiplier=0.15
}
